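# Stereo MetaSound source for the looping sound of the CrazyShapes vehicle boost.
# After a start delay, an ADSR envelope fades in a shuffled loop wave with random or defined start time and optional eased speed-based pitch.
# A switch selects a one-shot layer: speed-scaled repeat triggers or two players chained by delayed triggers into self-alternating crossfades, with random pitch and mono/stereo paths.
# The one-shot array is empty here.
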MetaSoundSource /VehicleAudio/Boosts/Boost_CrazyShapes/MSS_Vehicle_Boost_CrazyShapes_Loop_Preset
inputs [
 Boost.Loop.DelayTime {type=Time, access=Constructor}
 Boost.Loop.FadeTime {type=Time, access=Constructor}
 Boost.Loop.One Shot.TransitionType {type=Int32}
 Boost.Loop.OneShot.CrossfadeTime {type=Time, access=Constructor}
 Boost.Loop.OneShot.RandomPitch.Max {type=Float, access=Constructor}
 Boost.Loop.OneShot.RandomPitch.Min {type=Float, access=Constructor}
 Boost.Loop.OneShot.Stereo {type=Bool}
 Boost.Loop.OneShot.TriggerRate.Max {type=Time, access=Constructor}
 Boost.Loop.OneShot.TriggerRate.Min {type=Time, access=Constructor}
 Boost.Loop.OneShot.Volume {type=Float, access=Constructor}
 Boost.Loop.Oneshots {type=WaveAsset:Array, access=Constructor}
 Boost.Loop.StartTime.Defined {type=Time, access=Constructor}
 Boost.Loop.StartTime.Random {type=Bool}
 Boost.Loop.Volume {type=Float, access=Constructor}
 Boost.Loop.Waveasset {type=WaveAsset:Array, access=Constructor}
 Boost.Pitch.Max {type=Float, access=Constructor}
 Boost.Pitch.Min {type=Float, access=Constructor}
 Boost.PitchByNormalSpeed {type=Bool, access=Constructor}
 Fort.Vehicle.Acceleration {type=Float}
 Fort.Vehicle.BlowoffStrength {type=Float}
 Fort.Vehicle.BonusSpeed {type=Float}
 Fort.Vehicle.EnableShifting {type=Bool}
 Fort.Vehicle.Gear {type=Int32}
 Fort.Vehicle.IsGrounded {type=Bool}
 Fort.Vehicle.Load {type=Float}
 Fort.Vehicle.NormalSpeed {type=Float}
 Fort.Vehicle.OnBlowoff {type=Trigger}
 Fort.Vehicle.OnDownshift {type=Trigger}
 Fort.Vehicle.OnUpshift {type=Trigger}
 Fort.Vehicle.Pitch {type=Float}
 Fort.Vehicle.RPM {type=Float}
 Fort.Vehicle.Throttle {type=Float}
 UE.Source.OnPlay {type=Trigger}
 UE.Source.Orientation.Azimuth {type=Float}
 UE.Source.Orientation.Elevation {type=Float}
]
interfaces [Fort.Vehicle:1.0, UE.OutputFormat.Stereo:1.0, UE.Source.Orientation:1.0, UE.Source:1.1]
outputs [
 "UE.OutputFormat.Stereo.Audio:0" {type=Audio}
 "UE.OutputFormat.Stereo.Audio:1" {type=Audio}
]
pages [Default {
  input_defaults {
   Boost.Loop.DelayTime=0.1
   Boost.Loop.FadeTime=0.3
   Boost.Loop.One Shot.TransitionType=1
   Boost.Loop.OneShot.CrossfadeTime=1
   Boost.Loop.OneShot.RandomPitch.Max=2
   Boost.Loop.OneShot.RandomPitch.Min=-1
   Boost.Loop.OneShot.Stereo="<MetaSoundBuilderResult.SUCCEEDED: 0>"
   Boost.Loop.OneShot.TriggerRate.Max=0.1
   Boost.Loop.OneShot.TriggerRate.Min=0.216
   Boost.Loop.OneShot.Volume=0
   Boost.Loop.StartTime.Defined=0
   Boost.Loop.StartTime.Random="<MetaSoundBuilderResult.SUCCEEDED: 0>"
   Boost.Loop.Volume=-3
   Boost.Pitch.Max=1.5
   Boost.Pitch.Min=0
   Boost.PitchByNormalSpeed="<MetaSoundBuilderResult.SUCCEEDED: 0>"
   Fort.Vehicle.Acceleration=0
   Fort.Vehicle.BlowoffStrength=0
   Fort.Vehicle.BonusSpeed=0
   Fort.Vehicle.EnableShifting=null
   Fort.Vehicle.Gear=0
   Fort.Vehicle.IsGrounded="<MetaSoundBuilderResult.SUCCEEDED: 0>"
   Fort.Vehicle.Load=0
   Fort.Vehicle.NormalSpeed=0
   Fort.Vehicle.OnBlowoff=null
   Fort.Vehicle.OnDownshift=null
   Fort.Vehicle.OnUpshift=null
   Fort.Vehicle.Pitch=0
   Fort.Vehicle.RPM=0
   Fort.Vehicle.Throttle=0
   UE.Source.OnPlay=null
   UE.Source.Orientation.Azimuth=0
   UE.Source.Orientation.Elevation=0
   Boost.Loop.Oneshots []
   Boost.Loop.Waveasset [&/Boost_CrazyShapes_Loop_01]
  }
  nodes [
   n1 = VariableDeferredAccessor.Trigger(Variable=InitVariable.Trigger().Variable)
   n2 = UE.Add:Float(
    PrimaryOperand=MapRange.MapRange:Float(
     In=^Fort.Vehicle.NormalSpeed
     Out Range B=UE.ConversionTimeToFloat(In=^Boost.Loop.OneShot.TriggerRate.Max).Out
     Out Range A=UE.ConversionTimeToFloat(In=^Boost.Loop.OneShot.TriggerRate.Min).Out
    ).Out Value
    AdditionalOperands=UE.RandomFloat(Next=@n1.Value) {also {defaults {Max=0.1, Min=-0.1}}}.Value
   )
   TriggerDelay_1 = UE.Trigger Delay(Delay Time=^Boost.Loop.DelayTime, In=^UE.Source.OnPlay)
   ADSREnvelopeADSREnvelope_0 = ADSR Envelope.ADSR Envelope:Float(Attack Time=^Boost.Loop.FadeTime, Trigger Attack=@TriggerDelay_1.Out) {also {defaults {Sustain Level=1}}}
   n3 = /FortMetasound/Patches/MSP_SwitchOnInt_4(
    Int=^Boost.Loop.One Shot.TransitionType
    Compare=TriggerCompare.Clamp:Int32(Compare=@ADSREnvelopeADSREnvelope_0.On Attack Triggered, A=Array.Num:WaveAsset:Array(Array=^Boost.Loop.Oneshots).Num) {also {defaults {Type=1}}}.True
   )
   n4 = UE.TriggerRepeat(Period=UE.ConversionFloatToTime(In=Max.Max:Float(A=@n2.Out) {also {defaults {B=0.085}}}.Value).Out, Start=@n3.Out 0)
   TriggerSequence2_0 = "UE.Trigger Sequence (2)"(In=@n4.RepeatOut)
   n5 = Array.Shuffle:WaveAsset:Array(In Array=^Boost.Loop.Oneshots, Next=@TriggerSequence2_0.Out 1)
   n6 = UE.RandomFloat(Max=^Boost.Loop.OneShot.RandomPitch.Max, Min=^Boost.Loop.OneShot.RandomPitch.Min, Next=@n5.On Next)
   n7 = MapRange.MapRange:Float(Out Range B=^Boost.Pitch.Max, Out Range A=^Boost.Pitch.Min, In=/VehicleAudio/Patches/EaseInQuart(Input=^Fort.Vehicle.NormalSpeed).Output)
   n8 = TriggerCompare.Clamp:Bool(A=^Boost.PitchByNormalSpeed, Compare=^UE.Source.OnPlay) {also {defaults {B="<MetaSoundBuilderResult.SUCCEEDED: 0>"}}}
   n9 = Value.Value:Float(Reset=@n8.False, Set=@n8.True) {also {defaults {Target Value=1}}}
   n10 = VariableAccessor.Float(Variable=VariableMutator.Float(Variable=InitVariable.Float().Variable, Value=UE.Multiply:Float(PrimaryOperand=@n7.Out Value, AdditionalOperands=@n9.Output Value).Out).Variable)
   n11 = VariableAccessor.Float(Variable=@n10.Variable)
   n12 = UE.Add:Float(PrimaryOperand=@n6.Value, AdditionalOperands=@n11.Value)
   n13 = TriggerCompare.Clamp:Bool(A=^Boost.Loop.OneShot.Stereo, Compare=@n6.On Next)
   WavePlayer_0 = UE.Wave Player:Mono(Pitch Shift=@n12.Out, Wave Asset=@n5.Value, Play=@n13.True) {also {version [1,1]}}
   n14 = Array.Shuffle:WaveAsset:Array(In Array=^Boost.Loop.Oneshots, Next=@TriggerSequence2_0.Out 0)
   n15 = UE.RandomFloat(Max=^Boost.Loop.OneShot.RandomPitch.Max, Min=^Boost.Loop.OneShot.RandomPitch.Min, Next=@n14.On Next)
   n16 = VariableAccessor.Float(Variable=@n11.Variable)
   n17 = UE.Add:Float(PrimaryOperand=@n15.Value, AdditionalOperands=@n16.Value)
   n18 = TriggerCompare.Clamp:Bool(A=^Boost.Loop.OneShot.Stereo, Compare=@n15.On Next)
   WavePlayer_3 = UE.Wave Player:Mono(Pitch Shift=@n17.Out, Wave Asset=@n14.Value, Play=@n18.True) {also {version [1,1]}}
   AudioMixerAudioMixerMono2_0 = "AudioMixer.Audio Mixer (Mono, 2)"(In 1=@WavePlayer_0.Out Mono, In 0=@WavePlayer_3.Out Mono)
   WavePlayer_5 = UE.Wave Player:Stereo(Pitch Shift=@n12.Out, Wave Asset=@n5.Value, Play=UE.Reroute(Value=@n13.False).Value) {also {version [1,1]}}
   WavePlayer_7 = UE.Wave Player:Stereo(Pitch Shift=@n17.Out, Wave Asset=@n14.Value, Play=@n18.False) {also {version [1,1]}}
   AudioMixerAudioMixerStereo2_0 = "AudioMixer.Audio Mixer (Stereo, 2)"(In 1 L=@WavePlayer_5.Out Left, In 1 R=@WavePlayer_5.Out Right, In 0 L=@WavePlayer_7.Out Left, In 0 R=@WavePlayer_7.Out Right)
   AudioMixerAudioMixerStereo2_1 = "AudioMixer.Audio Mixer (Stereo, 2)"(
    In 0 L=@AudioMixerAudioMixerMono2_0.Out
    In 0 R=@AudioMixerAudioMixerMono2_0.Out
    In 1 L=@AudioMixerAudioMixerStereo2_0.Out L
    In 1 R=@AudioMixerAudioMixerStereo2_0.Out R
   )
   n19 = VariableDeferredAccessor.Trigger(Variable=InitVariable.Trigger().Variable)
   n20 = Array.Shuffle:WaveAsset:Array(In Array=^Boost.Loop.Oneshots, Next=@n19.Value)
   n21 = UE.RandomFloat(Max=^Boost.Loop.OneShot.RandomPitch.Max, Min=^Boost.Loop.OneShot.RandomPitch.Min, Next=@n20.On Next)
   n22 = VariableDeferredAccessor.Trigger(Variable=InitVariable.Trigger().Variable)
   TriggerAnyTriggerAny2_1 = "TriggerAny.Trigger Any (2)"(In 1=@n3.Out 1, In 0=@n22.Value)
   n23 = Array.Shuffle:WaveAsset:Array(In Array=^Boost.Loop.Oneshots, Next=@TriggerAnyTriggerAny2_1.Out)
   n24 = UE.RandomFloat(Max=^Boost.Loop.OneShot.RandomPitch.Max, Min=^Boost.Loop.OneShot.RandomPitch.Min, Next=@n23.On Next)
   ADSREnvelopeADSREnvelope_1 = ADSR Envelope.ADSR Envelope:Float(
    Attack Time=^Boost.Loop.OneShot.CrossfadeTime
    Release Time=^Boost.Loop.OneShot.CrossfadeTime
    Trigger Attack=@n21.On Next
    Trigger Release=@n24.On Next
   ) {also {defaults {Sustain Level=1}}}
   n25 = VariableAccessor.Float(Variable=VariableMutator.Float(Value=@ADSREnvelopeADSREnvelope_1.Out Envelope, Variable=InitVariable.Float().Variable).Variable)
   ADSREnvelopeADSREnvelope_2 = ADSR Envelope.ADSR Envelope:Float(
    Attack Time=^Boost.Loop.OneShot.CrossfadeTime
    Release Time=^Boost.Loop.OneShot.CrossfadeTime
    Trigger Release=@n21.On Next
    Trigger Attack=@n24.On Next
   ) {also {defaults {Sustain Level=1}}}
   n26 = VariableAccessor.Float(Variable=VariableMutator.Float(Value=@ADSREnvelopeADSREnvelope_2.Out Envelope, Variable=InitVariable.Float().Variable).Variable)
   n27 = TriggerCompare.Clamp:Bool(A=^Boost.Loop.OneShot.Stereo, Compare=@ADSREnvelopeADSREnvelope_1.On Attack Triggered)
   n28 = VariableAccessor.Float(Variable=@n16.Variable)
   n29 = VariableAccessor.Float(Variable=VariableMutator.Float(Value=UE.Add:Float(PrimaryOperand=@n21.Value, AdditionalOperands=@n28.Value).Out, Variable=InitVariable.Float().Variable).Variable)
   n30 = VariableAccessor.Float(Variable=@n29.Variable)
   WavePlayer_1 = UE.Wave Player:Mono(Wave Asset=@n20.Value, Play=@n27.True, Pitch Shift=@n30.Value) {also {version [1,1]}}
   n31 = TriggerCompare.Clamp:Bool(A=^Boost.Loop.OneShot.Stereo, Compare=@ADSREnvelopeADSREnvelope_2.On Attack Triggered)
   n32 = VariableAccessor.Float(Variable=@n28.Variable)
   n33 = VariableAccessor.Float(Variable=VariableMutator.Float(Value=UE.Add:Float(PrimaryOperand=@n24.Value, AdditionalOperands=@n32.Value).Out, Variable=InitVariable.Float().Variable).Variable)
   WavePlayer_2 = UE.Wave Player:Mono(Wave Asset=@n23.Value, Play=@n31.True, Pitch Shift=@n33.Value) {also {version [1,1]}}
   AudioMixerAudioMixerMono2_1 = "AudioMixer.Audio Mixer (Mono, 2)"(Gain 1=VariableAccessor.Float(Variable=@n25.Variable).Value, Gain 0=@n26.Value, In 1=@WavePlayer_1.Out Mono, In 0=@WavePlayer_2.Out Mono)
   WavePlayer_6 = UE.Wave Player:Stereo(Wave Asset=@n23.Value, Play=@n31.False, Pitch Shift=VariableAccessor.Float(Variable=@n33.Variable).Value) {also {version [1,1]}}
   WavePlayer_8 = UE.Wave Player:Stereo(Wave Asset=@n20.Value, Play=@n27.False, Pitch Shift=@n29.Value) {also {version [1,1]}}
   AudioMixerAudioMixerStereo2_4 = "AudioMixer.Audio Mixer (Stereo, 2)"(
    Gain 0=VariableAccessor.Float(Variable=@n26.Variable).Value
    Gain 1=@n25.Value
    In 0 L=@WavePlayer_6.Out Left
    In 0 R=@WavePlayer_6.Out Right
    In 1 L=@WavePlayer_8.Out Left
    In 1 R=@WavePlayer_8.Out Right
   )
   AudioMixerAudioMixerStereo2_3 = "AudioMixer.Audio Mixer (Stereo, 2)"(
    In 0 L=@AudioMixerAudioMixerMono2_1.Out
    In 0 R=@AudioMixerAudioMixerMono2_1.Out
    In 1 L=@AudioMixerAudioMixerStereo2_4.Out L
    In 1 R=@AudioMixerAudioMixerStereo2_4.Out R
   )
   AudioMixerAudioMixerStereo2_5 = "AudioMixer.Audio Mixer (Stereo, 2)"(
    In 0 L=@AudioMixerAudioMixerStereo2_1.Out L
    In 0 R=@AudioMixerAudioMixerStereo2_1.Out R
    In 1 L=@AudioMixerAudioMixerStereo2_3.Out L
    In 1 R=@AudioMixerAudioMixerStereo2_3.Out R
   )
   DecibelstoLinearGain_1 = UE.Decibels to Linear Gain:Float(Decibels=^Boost.Loop.Volume)
   n34 = VariableAccessor.Float(Variable=VariableMutator.Float(Value=@ADSREnvelopeADSREnvelope_0.Out Envelope, Variable=InitVariable.Float().Variable).Variable)
   DecibelstoLinearGain_0 = UE.Decibels to Linear Gain:Float(Decibels=^Boost.Loop.OneShot.Volume)
   n35 = Array.Shuffle:WaveAsset:Array(In Array=^Boost.Loop.Waveasset, Next=@ADSREnvelopeADSREnvelope_0.On Attack Triggered)
   n36 = UE.Reroute(Value=UE.Reroute(Value=@n35.Value).Value)
   GetWaveDuration_1 = "UE.Get Wave Duration: "(Wave=@n36.Value)
   n37 = TriggerCompare.Clamp:Bool(A=^Boost.Loop.StartTime.Random, Compare=@n35.On Next) {also {defaults {B="<MetaSoundBuilderResult.SUCCEEDED: 0>"}}}
   TriggerRouteTriggerRouteTime2_0 = "TriggerRoute.Trigger Route (Time, 2):Time"(
    Value 1=^Boost.Loop.StartTime.Defined
    Set 0=UE.RandomTime(Max=UE.Multiply:Time by Float(PrimaryOperand=@GetWaveDuration_1.Duration) {also {defaults {AdditionalOperands=0.95}}}.Out, Next=@n37.True).On Next
    Set 1=UE.Reroute(Value=UE.Reroute(Value=@n37.False).Value).Value
   ) {also {defaults {Value 0=0.7273}}}
   WavePlayer_4 = UE.Wave Player:Stereo(Wave Asset=UE.Reroute(Value=UE.Reroute(Value=@n36.Value).Value).Value, Play=@TriggerRouteTriggerRouteTime2_0.On Set, Pitch Shift=@n10.Value) {also {defaults {Loop="<MetaSoundBuilderResult.SUCCEEDED: 0>"}, version [1,1]}}
   AudioMixerAudioMixerStereo2_2 = "AudioMixer.Audio Mixer (Stereo, 2)"(
    In 1 L=@AudioMixerAudioMixerStereo2_5.Out L
    In 1 R=@AudioMixerAudioMixerStereo2_5.Out R
    Gain 0=UE.Multiply:Float(PrimaryOperand=@DecibelstoLinearGain_1.Linear Gain, AdditionalOperands=@n34.Value).Out
    Gain 1=UE.Multiply:Float(PrimaryOperand=@DecibelstoLinearGain_0.Linear Gain, AdditionalOperands=@n34.Value).Out
    In 0 L=@WavePlayer_4.Out Left
    In 0 R=@WavePlayer_4.Out Right
   )
   UE.ConversionTimeToFloat()
   TriggerRouteTriggerRouteTime2_1 = "TriggerRoute.Trigger Route (Time, 2):Time"()
   TriggerRouteTriggerRouteTime2_2 = "TriggerRoute.Trigger Route (Time, 2):Time"() {also {defaults {Value 0=0, Value 1=0}}}
   VariableAccessor.Float(Variable=@n30.Variable)
   GetWaveDuration_2 = "UE.Get Wave Duration: "(Wave=@n20.Value)
   n38 = Clamp.Clamp:Float(Max=UE.ConversionTimeToFloat(In=@GetWaveDuration_2.Duration).Out, In=UE.ConversionTimeToFloat(In=^Boost.Loop.OneShot.CrossfadeTime).Out) {also {defaults {Min=0}}}
   TriggerAnyTriggerAny2_2 = "TriggerAny.Trigger Any (2)"(In 0=UE.Reroute(Value=@WavePlayer_1.On Play).Value, In 1=@WavePlayer_8.On Play)
   TriggerDelay_2 = UE.Trigger Delay(
    Delay Time=UE.Subtract:Time(AdditionalOperands=UE.ConversionFloatToTime(In=@n38.Value).Out, PrimaryOperand=@GetWaveDuration_2.Duration) {also {version [1,1]}}.Out
    In=@TriggerAnyTriggerAny2_2.Out
   )
   VariableMutator.Trigger(Value=@TriggerDelay_2.Out, Variable=@n22.Variable)
   GetWaveDuration_0 = "UE.Get Wave Duration: "(Wave=@n23.Value)
   n39 = Clamp.Clamp:Float(Max=UE.ConversionTimeToFloat(In=@GetWaveDuration_0.Duration).Out, In=UE.ConversionTimeToFloat(In=^Boost.Loop.OneShot.CrossfadeTime).Out)
   TriggerAnyTriggerAny2_0 = "TriggerAny.Trigger Any (2)"(In 0=@WavePlayer_2.On Play, In 1=@WavePlayer_6.On Play)
   TriggerDelay_0 = UE.Trigger Delay(
    Delay Time=UE.Subtract:Time(AdditionalOperands=UE.ConversionFloatToTime(In=@n39.Value).Out, PrimaryOperand=@GetWaveDuration_0.Duration) {also {version [1,1]}}.Out
    In=@TriggerAnyTriggerAny2_0.Out
   )
   VariableMutator.Trigger(Value=@TriggerDelay_0.Out, Variable=@n19.Variable)
   VariableMutator.Trigger(Value=@n4.RepeatOut, Variable=@n1.Variable)
  ]
  links [
   {dst="@UE.OutputFormat.Stereo.Audio:0", src=AudioMixerAudioMixerStereo2_2.Out L}
   {dst="@UE.OutputFormat.Stereo.Audio:1", src=AudioMixerAudioMixerStereo2_2.Out R}
  ]
 }]
properties {
 assetImportData=&self:AssetImportData
 attenuationSettings=/VehicleAudio/Boosts/Vehicle_Boost_Att
 audioPropertiesSheet=/VehicleAudio/PropertySheets/APS_Cosmetic_Vehicle_Boost_Loop
 bLooping=true
 duration=10000
 maxDistance=5012
 numChannels=2
 outputFormat=Stereo
 sampleRate=48000
 soundClassObject=/Game/Sounds/FortSoundClassesAndModes/Fort_Vehicle_Exterior_Boost
 soundSubmixObject=&Fortnite_SFX
 concurrencySet [/VehicleAudio/Boosts/Vehicle_Boost_Loop_Conc]
 modulationSettings {
  highpassRouting=Union
  lowpassRouting=Union
  volumeRouting=Union
  highpassModulationDestination {bEnableModulation=true, modulators [/Game/Sounds/Modulation/Patches/HPF/PP_FortVehicles_Boost_HPF]}
  lowpassModulationDestination {bEnableModulation=true, modulators [/Game/Sounds/Modulation/Patches/LPF/PP_FortVehicles_Boost_LPF]}
  volumeModulationDestination {
   bEnableModulation=true
   modulators [/VehicleAudio/Modulation/DM_VehicleTurboPatch, /Game/Sounds/Modulation/Patches/PP_FortVehicles_Boost_Vol]
  }
 }
}
